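# Unlit, translucent, two-sided material for the Nebula skydiving contrail on Niagara ribbons and mesh particles.
# Panning Voronoi distortion warps a scrolling trail mask and flame shape, which feed opacity and curve-atlas color, added to gradient or particle color.
# Dynamic parameters drive dive blending and scroll speed; camera, view-angle and depth fades soften the trail.
# World position offset combines layered sine waves, camera offset, optional SplineThicken and optional noise-texture displacement along normals.
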Material /BRCosmetics/Athena/Effects/Particles/Skydiving/NS_Contrail_Nebula/M_Contrail_Nebula_Trail
graph {
 properties {
  bAutomaticallySetUsageInEditor=false
  bUsedWithNiagaraMeshParticles=true
  bUsedWithNiagaraRibbons=true
  blendMode=BLEND_Translucent
  pixelDepthOffsetMode=PDOM_Legacy
  shadingModel=MSM_Unlit
  twoSided=true
  usedShadingModels=Unlit
 }
 nodes [
  n1 = DynamicParameter() {parameterIndex=2, paramNames [SplineThicken_WidthMult, "", "", ""]}
  n2 = Multiply(A=ScalarParameter() {defaultValue=40, parameterName=SplineThicken_BaseWidth}, B=@n1.SplineThicken_WidthMult)
  n3 = Multiply(A=ScalarParameter() {defaultValue=80, parameterName=SplineThicken_TipWidth}, B=@n1.SplineThicken_WidthMult)
  n4 = StaticBool()
  n5 = Multiply(
   A=MaterialFunctionCall() {materialFunction=&CameraDirectionVector}.Result
   B=DynamicParameter() {defaultValue=<rgba 0.8 1 0 1>, paramNames [UV1 U Offset, Dive, Camera Offset, SpeedScale]}.Camera Offset
  )
  n6 = Add(
   A=TextureCoordinate()
   B=AppendVector(A=DynamicParameter() {defaultValue=<rgba 0.8 1 0 1>, paramNames [UV1 U Offset, Dive, Camera Offset, SpeedScale]}.UV1 U Offset, B=Constant())
  )
  n7 = Panner(Coordinate=@n6) {speedX=-1}
  n8 = ComponentMask(None=Panner(Coordinate=@n6) {speedX=-0.5}) {r=true}
  n9 = Add(A=Sine(None=ComponentMask(None=@n7) {r=true}) {period=1.4}, B=Multiply(A=Sine(None=@n8) {period=0.5}) {constB=0.3})
  n10 = Panner() {speedX=-0.75}
  n11 = Multiply(A=Sine(None=ComponentMask(None=@n10) {r=true}) {period=3}) {constB=0.7}
  n12 = VectorParameter() {group=WPO, parameterName=WPO Direction, defaultValue=<rgba 1 1 0 1>}
  n13 = TextureCoordinate()
  n14 = Multiply(
   A=Multiply(A=Add(A=@n9, B=@n11), B=Transform(None=@n12.RGB) {transformSourceType=TRANSFORMSOURCE_Instance})
   B=Power(Base=ComponentMask(None=@n13) {r=true})
  )
  Add(
   A=LinearInterpolate(
    B=MaterialFunctionCall(WidthBase=@n2, WidthTip=@n3, Expand U or V UV Channel=@n4) {materialFunction=&SplineThicken}.WorldPositionOffset
    Alpha=ScalarParameter() {parameterName=UseSplineThicken}
   )
   B=Add(A=Add(A=@n5, B=@n14), B=NamedRerouteUsage())
  )
  n15 = TextureCoordinate()
  n16 = AppendVector(
   A=ScalarParameter() {defaultValue=0.5, group=Texture_WPO, parameterName=Texture_WPO_X_Tile}
   B=ScalarParameter() {defaultValue=3, group=Texture_WPO, parameterName=Texture_WPO_Y_Tile}
  )
  n17 = Multiply(A=Time())
  n18 = ScalarParameter() {defaultValue=0.2, group=Texture_WPO, parameterName=Texture_WPO_NoiseA_SpeedX}
  n19 = ScalarParameter() {group=Texture_WPO, parameterName=Texture_WPO_NoiseA_SpeedY}
  n20 = TextureSampleParameter2D(UVs=Panner(Coordinate=Multiply(A=@n15, B=@n16), Time=@n17, Speed=AppendVector(A=@n18, B=@n19)) {speedX=0.2, speedY=1}) {group=Texture_WPO, parameterName=WPO_NoiseA, texture=&T_Inky_Smoke_Tile_Chromatic}
  n21 = TextureCoordinate()
  n22 = Panner(
   Coordinate=Multiply(A=@n16, B=Add(A=@n21))
   Time=@n17
   Speed=AppendVector(
    A=ScalarParameter() {defaultValue=-0.25, group=Texture_WPO, parameterName=Texture_WPO_NoiseB_SpeedX}
    B=ScalarParameter() {defaultValue=0.8, group=Texture_WPO, parameterName=Texture_WPO_NoiseB_SpeedY}
   )
  ) {speedX=-0.25, speedY=0.8}
  n23 = TextureSampleParameter2D(UVs=@n22) {group=Texture_WPO, parameterName=WPO_NoiseB, texture=&T_Inky_Smoke_Tile_Chromatic}
  LinearInterpolate(A=@n20.RGB, B=@n23.RGB)
  n24 = Constant2Vector() {r=0.5}
  n25 = Constant2Vector() {r=0.3}
  n26 = DynamicParameter() {parameterIndex=1, defaultValue=<rgba 0.5 1 1 0>, paramNames [Speed X, Speed Y, Speed Z, UseParticleColor]}
  n27 = TextureSample(UVs=Add(A=TextureCoordinate(), B=LinearInterpolate(A=@n24, B=@n25, Alpha=@n26.UseParticleColor) {constB=0})) {samplerSource=SSM_Clamp_WorldGroupSettings, texture=&/T_Gradient_Nebula}
  n28 = Multiply(A=@n27.RGB, B=Constant3Vector() {constant=<rgba 2 1.2 1 1>})
  n29 = Saturate(None=Multiply(A=LinearInterpolate(A=@n28, B=@n27.RGB, Alpha=@n26.UseParticleColor) {constB=0}) {constB=0.5})
  n30 = ParticleColor()
  n31 = TextureCoordinate()
  n32 = ScalarParameter() {defaultValue=2, parameterName=Vertical Gradient Tiling, sliderMax=3}
  n33 = DynamicParameter() {defaultValue=<rgba 0.8 1 0 1>, paramNames [UV1 U Offset, Dive, Camera Offset, SpeedScale]}
  n34 = Constant()
  n35 = ScalarParameter() {defaultValue=3, group=UV Distortion, parameterName=UV Speed X 2, sliderMax=5, sliderMin=-5}
  n36 = ScalarParameter() {group=UV Distortion, parameterName=UV Speed Y 2, sliderMax=5, sliderMin=-5}
  n37 = DynamicParameter() {defaultValue=<rgba 0.8 1 0 1>, paramNames [UV1 U Offset, Dive, Camera Offset, SpeedScale]}
  n38 = Panner(
   Coordinate=LinearInterpolate(A=TextureCoordinate() {coordinateIndex=1}, B=TextureCoordinate(), Alpha=ScalarParameter() {parameterName=UseTexCoord0})
   Speed=Multiply(A=AppendVector(A=@n35, B=@n36), B=@n37.SpeedScale)
  ) {speedX=-0.5}
  n39 = AppendVector(
   A=ScalarParameter() {defaultValue=0.25, group=UV Distortion, parameterName=UV Scale X 2, sliderMax=5}
   B=ScalarParameter() {defaultValue=0.5, group=UV Distortion, parameterName=UV Scale Y 2, sliderMax=5}
  )
  n40 = TextureSampleParameter2D(UVs=Add(A=AppendVector(A=@n33.UV1 U Offset, B=@n34), B=Multiply(A=@n38, B=@n39))) {
   group=UV Distortion
   parameterName=Distortion Texture
   texture=/BRCosmetics/Athena/Effects/Particles/Skydiving/NS_Contrail_PinkSpike/Textures/T_PinkSpike_Voronoi2D_Height
  }
  n41 = LinearInterpolate(A=TextureCoordinate() {coordinateIndex=1}, B=TextureCoordinate(), Alpha=ScalarParameter() {parameterName=UseTexCoord0})
  n42 = Multiply(
   A=@n37.SpeedScale
   B=AppendVector(
    A=ScalarParameter() {defaultValue=2, group=UV Distortion, parameterName=UV Speed X, sliderMax=5, sliderMin=-5}
    B=ScalarParameter() {group=UV Distortion, parameterName=UV Speed Y, sliderMax=5, sliderMin=-5}
   )
  )
  n43 = ScalarParameter() {defaultValue=0.4, group=UV Distortion, parameterName=UV Scale X, sliderMax=5}
  n44 = ScalarParameter() {defaultValue=0.8, group=UV Distortion, parameterName=UV Scale Y, sliderMax=5}
  n45 = Add(
   A=AppendVector(A=@n37.UV1 U Offset, B=Constant())
   B=Multiply(A=Panner(Coordinate=@n41, Speed=@n42) {speedX=-0.5}, B=AppendVector(A=@n43, B=@n44))
  )
  n46 = TextureSampleParameter2D(UVs=@n45) {
   group=UV Distortion
   parameterName=Distortion Texture
   texture=&T_PinkSpike_Voronoi2D_Height
  }
  n47 = MaterialFunctionCall(X=Multiply(A=@n40.R, B=@n46.R), Y=Multiply(A=@n40.G, B=@n46.G)) {materialFunction=&MakeFloat2}
  n48 = Subtract(A=OneMinus(None=Power(Base=@n47.Result))) {constB=0.25}
  n49 = ComponentMask(None=Add(A=Multiply(A=@n31, B=@n32), B=@n48)) {r=true}
  n50 = CurveAtlasRowParameter(CurveTime=Saturate(None=OneMinus(None=@n49))) {atlas=/Game/Athena/LinearColorCurves/CA_Nebula_Contrail, curve=/Game/Athena/LinearColorCurves/CC_Nebula_Contrail, parameterName=Curve_Color}
  n51 = Multiply(A=Multiply(A=@n50, B=@n50), B=ScalarParameter() {defaultValue=1, parameterName=Vertical Gradient Bloom, sliderMax=1})
  n52 = TextureCoordinate()
  n53 = SmoothStep(Value=OneMinus(None=ComponentMask(None=@n52) {r=true}))
  n54 = Add(A=TextureCoordinate(), B=Multiply(A=@n48) {constB=0.1})
  n55 = AppendVector(A=OneMinus(None=ComponentMask(None=@n54) {r=true}), B=ComponentMask(None=@n54) {g=true})
  n56 = ComponentMask(None=TextureCoordinate()) {r=true}
  n57 = AppendVector(A=Power(Base=OneMinus(None=@n56)), B=Constant() {r=1})
  n58 = Multiply(A=SphereMask(A=Multiply(A=@n55, B=@n57), B=Constant2Vector() {g=0.5, r=0.5}) {attenuationRadius=0.5, hardnessPercent=0}) {constB=3}
  n59 = ScalarParameter() {defaultValue=1, parameterName=UseFlameShape}
  n60 = Add(
   A=LinearInterpolate(A=TextureCoordinate() {coordinateIndex=1}, B=TextureCoordinate(), Alpha=ScalarParameter() {parameterName=UseTexCoord0})
   B=Multiply(A=@n48, B=ScalarParameter() {defaultValue=-0.2, group=UV Distortion, parameterName=Distortion Strength, sliderMax=1, sliderMin=-1}) {constB=0.1}
  )
  n61 = DynamicParameter() {defaultValue=<rgba 0.8 1 0 1>, paramNames [UV1 U Offset, Dive, Camera Offset, SpeedScale]}
  n62 = Multiply(
   A=AppendVector(
    A=ScalarParameter() {defaultValue=2.5, group=Flame, parameterName=Speed X, sliderMax=5, sliderMin=-5}
    B=ScalarParameter() {group=Flame, parameterName=Speed Y, sliderMax=5, sliderMin=-5}
   )
   B=@n61.SpeedScale
  )
  n63 = Panner(Coordinate=@n60, Speed=@n62) {speedX=1}
  n64 = AppendVector(
   A=ScalarParameter() {defaultValue=0.2, group=Flame, parameterName=Scale X, sliderMax=5}
   B=ScalarParameter() {defaultValue=1, group=Flame, parameterName=Scale Y, sliderMax=5}
  )
  n65 = AppendVector(A=@n61.UV1 U Offset, B=Constant() {r=0.2})
  n66 = TextureSampleParameter2D(UVs=Add(A=Multiply(A=@n63, B=@n64), B=@n65)) {
   parameterName=Mask Texture
   samplerSource=SSM_Wrap_WorldGroupSettings
   samplerType=SAMPLERTYPE_LinearGrayscale
   texture=&/T_Nebula_Trail
  }
  n67 = DynamicParameter() {parameterIndex=3, defaultValue=<rgba 1 1 0 1>, paramNames [DivePanSpeedMult, DiveLerp, "", ""]}
  n68 = Panner(Coordinate=@n60, Speed=Multiply(A=@n62, B=@n67.DivePanSpeedMult)) {speedX=1}
  n69 = AppendVector(
   A=ScalarParameter() {defaultValue=0.2, group=Flame, parameterName=Scale X, sliderMax=5}
   B=ScalarParameter() {defaultValue=1, group=Flame, parameterName=Scale Y, sliderMax=5}
  )
  n70 = TextureSampleParameter2D(UVs=Add(A=Multiply(A=@n68, B=@n69), B=@n65)) {
   parameterName=Mask Texture
   samplerSource=SSM_Wrap_WorldGroupSettings
   samplerType=SAMPLERTYPE_LinearGrayscale
   texture=&T_Nebula_Trail
  }
  n71 = TextureCoordinate() {uTiling=4}
  n72 = Saturate(None=OneMinus(None=ComponentMask(None=@n71) {r=true}))
  n73 = ComponentMask(None=@n71) {g=true}
  n74 = Multiply(A=Multiply(A=@n73, B=OneMinus(None=@n73))) {constB=4}
  n75 = LinearInterpolate(
   A=Multiply(A=LinearInterpolate(B=@n58, Alpha=@n59) {constA=1}, B=LinearInterpolate(A=@n66.R, B=@n70.R, Alpha=@n67.DiveLerp))
   B=@n58
   Alpha=Multiply(A=Multiply(A=@n72) {constB=0.5}, B=Saturate(None=@n74))
  )
  n76 = TextureCoordinate()
  n77 = Power(
   Base=Multiply(A=ComponentMask(None=@n76) {r=true}, B=ScalarParameter() {defaultValue=1, parameterName=BottomFade_Mult})
   Exp=ScalarParameter() {defaultValue=2, parameterName=BottomFade_Power}
  )
  n78 = LinearInterpolate(A=@n75, B=Multiply(A=@n75, B=Saturate(None=@n77)), Alpha=ScalarParameter() {parameterName=UseBottomFade})
  n79 = OneMinus(None=ComponentMask(None=TextureCoordinate()) {r=true})
  n80 = ScalarParameter() {defaultValue=2, parameterName=TopFadeMult}
  n81 = Saturate(None=Power(Base=Multiply(A=@n79, B=@n80) {constB=2}, Exp=ScalarParameter() {defaultValue=1, parameterName=TopFadePower}) {constExponent=1})
  n82 = MaterialFunctionCall(
   Fade Length=ScalarParameter() {defaultValue=50, parameterName=Camera Fade Length, sliderMax=200}
   Fade Offset=ScalarParameter() {defaultValue=35, parameterName=Camera Fade Offset, sliderMax=200}
  ) {materialFunction=&CameraDepthFade}
  n83 = AppendVector(A=AppendVector(A=@n26.Speed X, B=@n26.Speed X), B=@n26.Speed Z)
  n84 = MaterialFunctionCall() {materialFunction=&CameraDirectionVector}
  n85 = Power(Base=DotProduct(A=Normalize(VectorInput=@n83), B=@n84.Result), Exp=ScalarParameter() {defaultValue=1, parameterName=Testing}) {constExponent=4}
  n86 = LinearInterpolate(B=OneMinus(None=Saturate(None=@n85)), Alpha=ScalarParameter() {defaultValue=1, parameterName=UseCameraFacingOpacity}) {constA=1}
  n87 = ComponentMask(None=TextureCoordinate()) {r=true}
  n88 = DotProduct(A=@n84.Result, B=Constant3Vector() {constant=<rgba 0 0 1 0>})
  n89 = LinearInterpolate(B=Subtract(A=Power(Base=@n87) {constExponent=1}) {constB=0.1}, Alpha=Saturate(None=Power(Base=@n88))) {constA=1}
  n90 = ScalarParameter() {defaultValue=1, parameterName=UseCameraFacingOpacity}
  n91 = Multiply(
   A=LinearInterpolate(A=@n78, B=Multiply(A=@n81, B=@n78), Alpha=ScalarParameter() {parameterName=UseTopFade})
   B=Multiply(A=Multiply(A=@n82.Result, B=@n86), B=LinearInterpolate(B=@n89, Alpha=@n90) {constA=1}) {constA=1}
  )
  n92 = SmoothStep(Value=Multiply(A=Multiply(A=@n53, B=@n91), B=ScalarParameter() {defaultValue=1, parameterName=CurveTimeMult})) {constMax=1.5, constMin=0.1}
  n93 = CurveAtlasRowParameter(CurveTime=@n92) {atlas=&CA_Nebula_Contrail, curve=&CC_Nebula_Contrail, parameterName=Curve_Color}
  n94 = LinearInterpolate(
   A=ScalarParameter() {defaultValue=0.75, parameterName=Emissive, sliderMax=2}
   B=ScalarParameter() {defaultValue=1.2, parameterName=Emissive_Dive, sliderMax=3}
   Alpha=DynamicParameter() {defaultValue=<rgba 0.8 1 0 1>, paramNames [UV1 U Offset, Dive, Camera Offset, SpeedScale]}.Dive
  ) {constA=1}
  n95 = Multiply(A=@n93, B=Add(A=Multiply(A=@n92, B=@n94)) {constB=0.5})
  n96 = VectorParameter() {parameterName=TOD color, defaultValue=<rgba 1 1 1 1>}
  n97 = Multiply(
   A=Add(A=LinearInterpolate(A=@n29, B=@n30.RGB, Alpha=@n26.UseParticleColor) {constB=0}, B=Add(A=@n51, B=@n95))
   B=MaterialFunctionCall(Value_Morning=@n96.R, Value_Day=@n96.R, Value_Evening=@n96.R, Value_Night=@n96.R) {materialFunction=&Scalar_Time_of_Day}.Result
  )
  MaterialFunctionCall(Emissive=@n97) {materialFunction=&MF_FX_EmissiveAdjustments}
  n98 = Multiply(A=TextureSample(UVs=@n45) {texture=&T_PinkSpike_Voronoi2D_Height}.RGBA)
  MaterialFunctionCall(Float4=OneMinus(None=Power(Base=@n98))) {materialFunction=&BreakOutFloat4Components}
  MaterialFunctionCall(XY=@n22) {materialFunction=&Swizzle}
  Multiply(A=@n97)
  n99 = SmoothStep(Value=@n91) {constMax=0.55}
  n100 = Step(Y=@n91) {constX=0.7}
  n101 = Add(A=Multiply(A=@n99) {constB=0.8}, B=Multiply(A=OneMinus(None=@n100)) {constB=0.5})
  n102 = ScalarParameter() {defaultValue=0.8, parameterName=Opacity, sliderMax=1}
  n103 = ScalarParameter() {parameterName=AngleFade}
  n104 = Multiply(
   A=LinearInterpolate(A=@n99, B=Multiply(A=@n101, B=@n102), Alpha=ScalarParameter() {parameterName=OpaqueCenter, sliderMax=1})
   B=LinearInterpolate(
    B=MaterialFunctionCall("FadeSensitivity (optional)"=@n103) {materialFunction=&FadeBasedOnViewAngle}.Result
    Alpha=ScalarParameter() {parameterName=UseAngleFade}
   ) {constA=1}
  )
  n105 = DepthFade(Opacity=@n104, FadeDistance=ScalarParameter() {defaultValue=20, parameterName=Depth Fade Distance, sliderMax=100}) {fadeDistanceDefault=25}
  Multiply(
   A=Multiply(A=ShadingPathSwitch(Default=@n105, Deferred=@n105, Forward=@n104, Mobile=@n104), B=@n30.A)
   B=ScalarParameter() {defaultValue=0.8, parameterName=Opacity, sliderMax=1}
  )
  n106 = Multiply(A=Multiply(A=VertexNormalWS(), B=@n20.RGB), B=ScalarParameter() {group=Texture_WPO, parameterName=TextureWPO}) {constB=7}
  n107 = MaterialFunctionCall(Float2=TextureCoordinate()) {materialFunction=&BreakOutFloat2Components}
  n108 = Saturate(None=Multiply(A=Power(Base=@n107.R) {constExponent=1.5}) {constB=2})
  NamedRerouteDeclaration(None=LinearInterpolate(B=Multiply(A=@n106, B=@n108), Alpha=ScalarParameter() {group=Texture_WPO, parameterName=UseTextureWPO})) {name=TextureWPO, nodeColor=<rgba 0 0.7422 1 1>}
  ScalarParameter() {parameterName=Camera offset}
  SmoothStep(Value=@n91) {constMax=0.8}
  TextureCoordinate() {coordinateIndex=1}
  TextureCoordinate() {coordinateIndex=1}
 ]
}
