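# Placeholder lit moon-rock material for Niagara mesh particles in the Nebula skydiving contrail.
# Samples borrowed MoonLanding diffuse, normal and mask textures at half UV tiling.
# A blue-violet glow term sums a Fresnel rim and a downward world-normal gradient; the file shows no output pin connections.
# Dynamic parameter index 3 drives the rim exponent (F_Expo) and the gradient's scale (Z_Offset, Intensity), leaving rim brightness fixed.
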Material /BRCosmetics/Athena/Effects/Particles/Skydiving/NS_Contrail_Nebula/M_Moon_Rocks_Nebula_PlaceHolder
graph {
 properties {
  bUsedWithNiagaraMeshParticles=true
  pixelDepthOffsetMode=PDOM_Legacy
  previewMesh=/BRCosmetics/Athena/Effects/Particles/Skydiving/NS_Contrail_Avalanche/FX/SM_Chunk_02
  usedShadingModels=Default Lit
 }
 nodes [
  n1 = DynamicParameter() {parameterIndex=3, defaultValue=<rgba 0.5 1 5 1>, paramNames [Intensity,Z_Offset,F_Expo,Param4]}
  n2 = Constant3Vector() {constant=<rgba 0.1543 0 2 1>}
  n3 = DotProduct(A=VertexNormalWS(), B=Multiply(A=Constant3Vector() {constant=<rgba 0 0 -1 0>}, B=@n1.Z_Offset) {constB=0.5})
  Add(
   A=Multiply(A=Fresnel(ExponentIn=@n1.F_Expo) {exponent=3}, B=Multiply(A=@n2) {constB=1.5})
   B=Multiply(A=@n2, B=Multiply(A=@n3, B=@n1.Intensity) {constB=0.2})
  )
  n4 = Multiply(A=TextureCoordinate(), B=Constant2Vector() {g=0.5, r=0.5})
  TextureSample(UVs=@n4) {texture=/Game/Environments/MoonLanding/Materials/Textures/Moonrock_D}
  TextureSample(UVs=@n4) {samplerType=SAMPLERTYPE_Normal, texture=/Game/Environments/MoonLanding/Materials/Textures/Moonrock_N}
  TextureSample(UVs=@n4) {samplerType=SAMPLERTYPE_Masks, texture=/Game/Environments/MoonLanding/Materials/Textures/Moonrock_S}
 ]
}
