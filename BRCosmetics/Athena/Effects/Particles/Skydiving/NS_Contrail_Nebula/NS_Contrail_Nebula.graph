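# Nebula skydiving contrail Niagara system that dresses the character's hands and feet in purple auras, lights, ribbons, rocks and sparkles.
# Emitters sample skeletal mesh sockets, with moon rocks and sparkles lerping from offset spawn points back toward them.
# Flame ribbons follow Rock_Moons particles through an attribute reader, with spawn rate scaled by platform quality.
# Dive state, paraglide lean and front-end modifiers reshape trail velocities and spawn offsets.
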NiagaraSystem /BRCosmetics/Athena/Effects/Particles/Skydiving/NS_Contrail_Nebula/NS_Contrail_Nebula
dependencies {
 usedDataInterfaces [
  &NiagaraDataInterfaceCurve
  &NiagaraDataInterfaceColorCurve
  &NiagaraDataInterfaceArrayFloat
  &NiagaraDataInterfaceParticleRead
  &NiagaraDataInterfacePlatformSet
 ]
 usedDynamicInputs [&MakeFloatFromVector]
 usedModules [
  &self:SystemSpawnScript.SystemScriptSource.SystemScriptGraph.NiagaraNodeAssignment_6.SetVariables_#1
  &self:SystemSpawnScript.SystemScriptSource.SystemScriptGraph.NiagaraNodeAssignment_7.SetVariables_#2
  &self:SystemSpawnScript.SystemScriptSource.SystemScriptGraph.NiagaraNodeAssignment_3.SetVariables_#3
  &self:CalculateRotationalSpeed
  &self:SystemSpawnScript.SystemScriptSource.SystemScriptGraph.NiagaraNodeAssignment_4.SetVariables_#4
  /Game/Effects/Niagara/Modules/Skydiving/FN_DivingAttributes/v5/FN_DivingAttributes
  &self:SystemSpawnScript.SystemScriptSource.SystemScriptGraph.NiagaraNodeAssignment_0.SetVariables_#5
  &self:SystemSpawnScript.SystemScriptSource.SystemScriptGraph.NiagaraNodeAssignment_1.SetVariables_#6
  &SystemState
  &SampleSkeletalMesh
 ]
}
emitters {
 Flames_Attached_to_Rocks {
  emitterUpdateScript [
   &EmitterState()
   &SpawnParticlesFromOtherEmitter(
    Attribute Reader=DataInterface() {propertyValues {EmitterBinding {bindingMode=Other, emitterName=Rock_Moons}}}
    Spawn Rate=&MakeCustomFloatFromBool(
     Bool=&IsPlatformSetActive(New Platform Set=DataInterface() {propertyValues {Platforms {qualityLevelMask=-31}}})
     True Float=&Lerp_Float(A=32, B=64, Alpha=$User.ParaGlideLeanAlpha)
     False Float=&MakeCustomFloatFromBool(
      Bool=&IsPlatformSetActive(New Platform Set=DataInterface() {propertyValues {Platforms {qualityLevelMask=-30}}})
      True Float=&Lerp_Float(A=64, B=128, Alpha=$User.ParaGlideLeanAlpha)
      False Float=&Lerp_Float(A=128, B=256, Alpha=$User.ParaGlideLeanAlpha)
     )
    )
    Spawn Rate Per Particle Cap=10000000
    Use Maximum Age=false
    Use Minimum Age=false
   )
  ]
  particleSpawnScript [
   &InitializeParticle(UsePositionOffset=false, Write Material Random=true)
   &InitializeRibbon(
    Color=<rgba 1 1 1 0.9>
    Lifetime=0.25
    Position Mode:ENiagara_PositionInitializationMode=::NewEnumerator0
    Ribbon Facing Mode:ENiagara_UnsetDirectSet=::NewEnumerator1
    Ribbon Facing Vector:Vector3f=<1 0 0>
    Ribbon Width=30
    Write Material Random=true
   )
   &SampleParticlesFromOtherEmitter(
    Applied Position Offset:Vector3f=<0 0 -10>
    Applied Velocity Scale=0
    Particle ID Sampling=::NewEnumerator1
    Sprite Size Sampling:ENiagara_DirectReadSamplingMode=::NewEnumerator1
    Velocity Sampling=::NewEnumerator0
   )
   /SetVariables(Particles.RibbonWidth=&Multiply_Float(A=&MakeFloatFromVector2d(Vector 2D=$Output.SampleParticlesFromOtherEmitter.SampledSpriteSize), B=1.25)) {isSetParameters=true}
   &AddVelocity(Velocity=&FN_FrontEndVector3Modifier(Gameplay=<0 0 250>, Front End=&Lerp_Vector(A=<0 0 250>, B=<0 -500 250>, Alpha=$User.ParaGlideLeanAlpha)))
  ]
  particleUpdateScript [
   &InheritSourceMovement(
    Applied Position Scale=<-0.15 -0.15 0>
    ApplyForce=false
    ApplyPosition=true
    ApplyPositionFalloff=false
    Source Velocity:Vector3f=&TransformVector(Vector=$Engine.Owner.Velocity, Source Space=::World, Destination Space=::Simulation)
   )
   &self:ReadRibbonSourceParticleAge_04(PAR=DataInterface() {propertyValues {EmitterBinding {bindingMode=Other, emitterName=Rock_Moons}}}) {module=ReadRibbonSourceParticleAge_04}
   &SolveForcesAndVelocity(Clamp Velocity=false, Limit Acceleration=false)
   &ScaleColor(
    Linear Color Curve=DataInterface() {propertyValues {
      AlphaCurve {keys [<at 0.7287 1>,<at 1 0>]}
      BlueCurve {keys [<at 0 0.2>,<at 0.3679 0.4>,<at 0.9715 0.4>]}
      GreenCurve {keys [<at 0 0.0352>,<at 0.3679 0.0727>,<at 0.9715 0.1646>]}
      RedCurve {keys [<at 0 0.0391>,<at 0.3679 0.079>,<at 0.9715 0.2438>]}
     }}
    Scale Mode=::NewEnumerator2
   ) {module=ScaleColor001}
   &DynamicMaterialParameters(
    Index 1 Param 4=&Lerp_Float(
     A=0
     B=1
     Alpha=&FloatFromCurve(
      FloatCurve=DataInterface() {propertyValues {Curve {keys [
          {interpMode=RCIM_Cubic, tangentMode=RCTM_User, time=0, value=0}
          {arriveTangent=2.2243, interpMode=RCIM_Cubic, leaveTangent=2.2243, tangentMode=RCTM_User, time=1, value=1}
         ]}}}
      CurveIndex=$Particles.SourceParticleAge
     )
    )
    Module.Param0WriteEnabled1=false
    Module.Param1WriteEnabled1=false
    Module.Param2WriteEnabled1=false
    Module.Param3WriteEnabled1=true
    Write Parameter Index 0=false
    Write Parameter Index 1=true
   ) {module=DynamicMaterialParameters001}
   &ParticleState()
   &ScaleColor(
    Scale Alpha=&FloatFromCurve(
     FloatCurve=DataInterface() {propertyValues {Curve {keys [
         {tangentMode=RCTM_User, time=0, value=1}
         {arriveTangent=-0.0405, interpMode=RCIM_Cubic, leaveTangent=-0.0405, tangentMode=RCTM_User, time=0.4, value=1}
         {arriveTangent=-1.6667, interpMode=RCIM_Cubic, leaveTangent=-1.6667, time=0.7662, value=0.797}
         {arriveTangent=-8.1179, interpMode=RCIM_Cubic, leaveTangent=-8.1179, tangentMode=RCTM_User, time=1, value=0}
        ]}}}
     CurveIndex=$Particles.SourceParticleAge
    )
    ScaleA=true
    ScaleRGB=true
   )
  ]
  rendererClasses [&NiagaraRibbonRendererProperties]
  renderers [{rendererClass=&NiagaraRibbonRendererProperties}]
 }
 Flames_Attached_to_Rocks002 {
  emitterUpdateScript [
   &EmitterState()
   &SpawnRate(SpawnRate=120, Use Spawn Probability=false)
  ]
  particleSpawnScript [
   &InitializeRibbon(
    Lifetime=0.2
    Position Mode:ENiagara_PositionInitializationMode=::NewEnumerator0
    Ribbon Facing Mode:ENiagara_UnsetDirectSet=::NewEnumerator1
    Ribbon Facing Vector:Vector3f=<1 0 0>
    Ribbon Width=25
    Write Material Random=true
   )
   &FN_SetSpawnGroup(Number of Spawn Groups=2)
   &SkeletalMeshLocation(
    Bone / Socket Index=&Add_Integer(A=$Particles.SpawnGroup, B=2)
    Bone Sampling Mode=::NewEnumerator4
    MaskSpawnGroup=false
    Sampled Position Offset=<0 0 -5>
    Sampled Position Offset Coordinate Space:ENiagaraSkelMeshTransforms=::NewEnumerator2
    Skeletal Mesh=$User.SkeletalMesh
   )
   &DynamicMaterialParameters(
    Index 1 Param 4=1
    Module.Param0WriteEnabled1=false
    Module.Param1WriteEnabled1=false
    Module.Param2WriteEnabled1=false
    Module.Param3WriteEnabled1=true
    Write Parameter Index 0=false
    Write Parameter Index 1=true
   ) {module=DynamicMaterialParameters001}
   &AddVelocity(Velocity=&FN_FrontEndVector3Modifier(
     Gameplay=&Lerp_Vector(
      A=<0 100 250>
      B=&SelectVectorFromArray(
       Direct Array Index=$Particles.SpawnGroup
       Vector Selection Array=DataInterface() {propertyValues {FloatData [<100 100 500>, <-100 100 500>]}}
      )
      Alpha=$System.Dive
     )
     Front End=&Multiply_VectorByFloat(
      Vector=&NormalizeVector(
       Vector To Normalize=&Lerp_Vector(A=<0 0.2 -1>, B=<0 2 -1>, Alpha=$User.ParaGlideLeanAlpha)
       Safe Normalize=true
       Safe Fallback Threshold=0
       Safe Fallback Vector=<0 0 0.001>
      )
      Float=&Lerp_Float(A=-300, B=-600, Alpha=$User.ParaGlideLeanAlpha)
     )
    ))
  ]
  particleUpdateScript [
   &InheritSourceMovement(
    Applied Position Scale=&Lerp_Vector(A=<1 1 1>, B=<0.99 0.99 1>, Alpha=$System.Dive)
    ApplyForce=false
    ApplyPosition=true
    ApplyPositionFalloff=false
   )
   &SolveForcesAndVelocity(Clamp Velocity=false, Limit Acceleration=false)
   &DynamicMaterialParameters(
    Index 0 Param 1=&SelectFloatFromArray(Direct Array Index=$Particles.SpawnGroup, Float Selection Array=DataInterface() {propertyValues {FloatData [0.2,0.1]}})
    Index 0 Param 2=$System.Dive
    Index 0 Param 4=&MakeCustomFloatFromBool(
     Bool=&SetBoolByFloatComparison(A=$System.Dive, B=1, Comparison Type=::NewEnumerator4)
     True Float=1.4
     False Float=0.6
    )
    Module.Param0WriteEnabled=true
    Module.Param1WriteEnabled=true
    Module.Param2WriteEnabled=false
    Module.Param3WriteEnabled=true
   )
   &ParticleState()
   &ScaleColor(
    Linear Color Curve=DataInterface() {propertyValues {
      AlphaCurve {keys [<at 0.719 1>,<at 1 0>]}
      BlueCurve {keys [<at 0 0.2>,<at 0.3679 0.4>,<at 0.9715 0.4>]}
      GreenCurve {keys [<at 0 0.0407>,<at 0.3679 0.0945>,<at 0.9715 0.1646>]}
      RedCurve {keys [<at 0 0.0352>,<at 0.3679 0.0727>,<at 0.9715 0.2438>]}
     }}
    Scale Mode=::NewEnumerator2
   )
  ]
  rendererClasses [&NiagaraRibbonRendererProperties]
  renderers [{rendererClass=&NiagaraRibbonRendererProperties}]
 }
 Hand_Rocks {
  emitterUpdateScript [
   &EmitterState(Life Cycle Mode=::NewEnumerator1, Loop Behavior=::NewEnumerator1, Loop Duration Mode=::NewEnumerator1, UseLoopDelay=false)
   &SpawnBurst_Instantaneous(Spawn Count=2, Spawn Group=1, Use Spawn Probability=false, UseLoopCountLimit=false) {module=SpawnBurst_Instantaneous001}
   &SpawnBurst_Instantaneous(Spawn Count=2, Use Spawn Probability=false, UseLoopCountLimit=false)
  ]
  particleSpawnScript [
   /SetVariables(Particles.Color_Sprite:LinearColor=<rgba 0.0611 0 0.702 2>, Particles.Size_Sprite:Vector2f=<50 50>) {isSetParameters=true}
   /SetVariables(Particles.VisibilityTag=$Emitter.SpawnGroup) {isSetParameters=true}
   /SetVariables(Particles.isFootAura:NiagaraBool=&SetBoolByIntComparison(A=$Emitter.SpawnGroup, B=0, Comparison Type=::NewEnumerator0)) {isSetParameters=true}
   &FN_SetSpawnGroup(Number of Spawn Groups=2)
   &InitializeParticle(
    Color=<rgba 1 1 1 1.5>
    Lifetime=1.5
    Mesh Scale Mode=::NewEnumerator3
    Mesh Uniform Scale=0.25
    UsePositionOffset=false
    Write Material Random=true
   )
   &SpriteFacingAndAlignment(WriteAlignment=false, WriteFacing=true)
   /SetVariables(Particles.AuraSize:Vector2f=<25 25>) {isSetParameters=true}
  ]
  particleUpdateScript [
   &ParticleState(
    Kill Particles When Lifetime Has Elapsed=false
    Let Infinitely Lived Particles Die When Emitter Deactivates=true
    Loop Particles Lifetime=true
   )
   &SkeletalMeshLocation(
    Bone / Socket Index=&MakeCustomIntFromBool(Bool=$Particles.isFootAura, True INT=&Add_Integer(A=$Particles.SpawnGroup, B=2), False INT=$Particles.SpawnGroup)
    Bone Sampling Mode=::NewEnumerator4
    MaskSpawnGroup=false
    Sampled Position Offset=&MakeCustomVectorFromBool(
     Bool=$Particles.isFootAura
     True Vector=<0 0 0>
     False Vector=&AddVectorToPosition(
      Position=&RandomRangeVector(
       Minimum=&VectorFromFloat(Value=&FloatFromCurve(FloatCurve=DataInterface() {propertyValues {Curve {keys [
             <at 0 1 RCIM_Cubic>
             <at 0.25 0 RCIM_Cubic>
             {interpMode=RCIM_Cubic, tangentMode=RCTM_User, time=0.5, value=1}
             {interpMode=RCIM_Cubic, tangentMode=RCTM_User, time=0.75, value=0}
             {tangentMode=RCTM_User, time=1, value=1}
            ]}}}))
       Maximum=&VectorFromFloat(Value=&FloatFromCurve(FloatCurve=DataInterface() {propertyValues {Curve {keys [
             {interpMode=RCIM_Cubic, tangentMode=RCTM_User, time=0, value=0}
             {interpMode=RCIM_Cubic, tangentMode=RCTM_User, time=0.5, value=1}
             {tangentMode=RCTM_User, time=1, value=0}
            ]}}}))
      )
      Vector=&SelectVectorFromArray(Direct Array Index=$Particles.SpawnGroup, Vector Selection Array=DataInterface() {propertyValues {FloatData [<8 -12 0>, <-8 10 0>]}})
     )
    )
    Skeletal Mesh=$User.SkeletalMesh
   )
   /SetVariables(Particles.Bowl_Position:NiagaraPosition=&AddVectorToPosition(
     Position=$Output.SkeletalMeshLocation.SampledPositionWithoutOffset
     Vector=&SelectVectorFromArray(Direct Array Index=$Particles.SpawnGroup, Vector Selection Array=DataInterface() {propertyValues {FloatData [<15 0 -15>, <15 0 -15>]}})
    )) {isSetParameters=true}
   /Niagara/Modules/Update/Orientation/V2/OrientMeshToVector(
    Look At Direction:Vector3f=&FN_FrontEndVector3Modifier(
     Gameplay=&NormalizeVector(
      Vector To Normalize=&TransformVector(Vector=$Engine.Owner.Velocity, Destination Space=::Local)
      Safe Normalize=true
      Safe Fallback Threshold=0
      Safe Fallback Vector=<0 0 0.001>
     )
     Front End=&TransformVector(Vector=&Lerp_Vector(A=<0 0.4 -1>, B=<0 2 -1>, Alpha=$User.ParaGlideLeanAlpha), Destination Space=::Simulation)
    )
    Override Incoming Orientation=1
    Override Quat=false
    Reference Axis Vector:Vector3f=<-1 0 0>
    UseOptionalSideAxis=false
    UseOptionalUpAxis=false
   )
   /SetVariables(Particles.Color_Sprite:LinearColor=&ScaleLinearColorByCurve(
     InitialValue=$Particles.Initial.Color_Sprite
     ScaleCurve=DataInterface() {propertyValues {
       AlphaCurve {keys [<at 0.0069 0.5>,<at 0.2467 0.8>,<at 0.4952 0.5>,<at 0.7503 0.8>,<at 1 0.5>]}
       BlueCurve {keys [<at 0 0.7>,<at 0.4777 0.7>,<at 0.9954 0.7>]}
       GreenCurve {keys [<at 0 0.7>,<at 0.4777 0.2775>,<at 0.9954 0.7>]}
       RedCurve {keys [<at 0 0.7>,<at 0.4777 0.3391>,<at 0.9954 0.7>]}
      }}
     CurveIndex=$Particles.NormalizedAge
    )) {isSetParameters=true}
   /SetVariables(Particles.Bowl_Size:Vector3f=&MakeVector(X=0.25, Y=0.25, Z=0.25)) {isSetParameters=true}
   /SetVariables(Particles.CameraOffset_Bowl=3) {isSetParameters=true}
   &SolveForcesAndVelocity(Clamp Velocity=false, Limit Acceleration=false)
   &FN_Light_Attributes(
    Color=<rgba 1.7553 0 25 1>
    Create Color Attribute=true
    Create Exponent Attribute=true
    Create Radius Attribute=true
    Enable Particle Light Enable=true
    Exponent=5
    Light Enabled=true
    Radius=60
    Scale Color=2.5
   )
   /SetVariables(Particles.AuraColor:LinearColor=&ScaleLinearColorByCurve(
     InitialValue=$Particles.Initial.AuraColor
     ScaleCurve=DataInterface() {propertyValues {
       AlphaCurve {keys [<at 0.0069 0.5>,<at 0.2467 0.8>,<at 0.4952 0.5>,<at 0.7503 0.8>,<at 1 0.5>]}
       BlueCurve {keys [<at 0 0.7>,<at 0.2501 0.7>,<at 0.5002 0.6>,<at 0.7428 0.75>,<at 1 0.7>]}
       GreenCurve {keys [<at 0 0>,<at 0.2501 0>,<at 0.5002 0.0098>,<at 0.7428 0>,<at 1 0>]}
       RedCurve {keys [<at 0 0.0453>,<at 0.2501 0.0364>,<at 0.5002 0.065>,<at 0.7428 0.039>,<at 1 0.0467>]}
      }}
     CurveIndex=$Particles.NormalizedAge
    )) {isSetParameters=true}
  ]
  rendererClasses [
   &NiagaraMeshRendererProperties
   &NiagaraSpriteRendererProperties
   &NiagaraLightRendererProperties
  ]
  renderers [
   {rendererClass=&NiagaraMeshRendererProperties, rendererIndex=0}
   {rendererClass=&NiagaraSpriteRendererProperties, rendererIndex=1}
   {rendererClass=&NiagaraMeshRendererProperties, rendererIndex=2}
   {rendererClass=&NiagaraLightRendererProperties, rendererIndex=3}
   {rendererClass=&NiagaraSpriteRendererProperties, rendererIndex=4}
  ]
 }
 MergedRibbons {
  emitterUpdateScript [
   &EmitterState()
   &SpawnRate(SpawnRate=120, Use Spawn Probability=false)
   &SpawnRate(SpawnGroup=1, SpawnRate=120, Use Spawn Probability=false) {enabled=false, module=SpawnRate001}
  ]
  particleSpawnScript [
   /SetVariables(Particles.SpawnGroup=$Emitter.SpawnGroup) {enabled=false, isSetParameters=true}
   &PartitionParticles()
   /SetVariables(Particles.isFootRibbon:NiagaraBool=&SetBoolByIntComparison(A=$Particles.PartitionParticles.Partition, B=1, Comparison Type=::NewEnumerator0)) {isSetParameters=true}
   /SetVariables(Particles.UsableIndex:NiagaraInt32=&MergeHelper_IntSwitch(
     0=$Particles.PartitionParticles.Partition
     1=&Add_Integer(A=$Particles.PartitionParticles.Partition, B=2)
     2=?
     3=?
     4=?
     5=?
     Index=$Particles.SpawnGroup
     Num=::NewEnumerator0
    )) {enabled=false, isSetParameters=true}
   &InitializeParticle(
    Lifetime=&MergeHelper_FloatSwitch(0=1, 1=1, 2=0.2, 3=0.2, Index=$Particles.PartitionParticles.Partition, Num=::NewEnumerator2)
    UsePositionOffset=false
    Write Material Random=true
   )
   &InitializeRibbon(
    Lifetime=&MakeCustomFloatFromBool(Bool=$Particles.isFootRibbon, True Float=0.2, False Float=0.4)
    Position Mode:ENiagara_PositionInitializationMode=::NewEnumerator0
    Ribbon Facing Mode:ENiagara_UnsetDirectSet=::NewEnumerator1
    Ribbon Facing Vector:Vector3f=<1 0 0>
    Ribbon Width=&MakeCustomFloatFromBool(Bool=$Particles.isFootRibbon, True Float=25, False Float=50)
    Write Material Random=true
   )
   &FN_SetSpawnGroup(Number of Spawn Groups=2) {enabled=false}
   &AddVelocity(Velocity=&FN_FrontEndVector3Modifier(
     Gameplay=&Lerp_Vector(
      A=<0 100 250>
      B=&SelectVectorFromArray(
       Direct Array Index=$Particles.PartitionParticles.Partition
       Vector Selection Array=DataInterface() {propertyValues {FloatData [<100 150 500>, <-100 150 500>]}}
      )
      Alpha=$System.Dive
     )
     Front End=&Multiply_VectorByFloat(
      Vector=&NormalizeVector(
       Vector To Normalize=&Lerp_Vector(A=<0 0.2 -1>, B=<0 2 -1>, Alpha=$User.ParaGlideLeanAlpha)
       Safe Normalize=true
       Safe Fallback Threshold=0
       Safe Fallback Vector=<0 0 0.001>
      )
      Float=&Lerp_Float(A=-300, B=-600, Alpha=$User.ParaGlideLeanAlpha)
     )
    ))
   &DynamicMaterialParameters(
    Index 1 Param 4=1
    Module.Param0WriteEnabled1=false
    Module.Param1WriteEnabled1=false
    Module.Param2WriteEnabled1=false
    Module.Param3WriteEnabled1=true
    Write Parameter Index 0=false
    Write Parameter Index 1=true
   ) {module=DynamicMaterialParameters001}
   &SkeletalMeshLocation(
    Bone / Socket Index=$Particles.PartitionParticles.Partition
    Bone Sampling Mode=::NewEnumerator4
    MaskSpawnGroup=false
    Sampled Position Offset=&SelectVectorFromArray(
     Direct Array Index=$Particles.PartitionParticles.Partition
     Vector Selection Array=DataInterface() {propertyValues {FloatData [<0 -5 0>, <0 5 0>, <-5 0 0>, <5 0 0>]}}
    )
    Skeletal Mesh=$User.SkeletalMesh
   )
  ]
  particleUpdateScript [
   &InheritSourceMovement(
    Applied Position Scale=&Lerp_Vector(A=<1 1 1>, B=<0.99 0.99 0.99>, Alpha=$System.Dive)
    ApplyForce=false
    ApplyPosition=true
    ApplyPositionFalloff=false
   )
   &SolveForcesAndVelocity(Clamp Velocity=false, Limit Acceleration=false)
   &DynamicMaterialParameters(
    Index 0 Param 1=&SelectFloatFromArray(
     Direct Array Index=$Particles.PartitionParticles.Partition
     Float Selection Array=DataInterface() {propertyValues {FloatData [0.8,0.6,0.2,0.1]}}
    )
    Index 0 Param 2=$System.Dive
    Index 0 Param 4=&MakeCustomFloatFromBool(
     Bool=$Particles.isFootRibbon
     True Float=&MakeCustomFloatFromBool(
      Bool=&SetBoolByFloatComparison(A=$System.Dive, B=1, Comparison Type=::NewEnumerator4)
      True Float=1.4
      False Float=0.6
     )
     False Float=&MakeCustomFloatFromBool(
      Bool=&SetBoolByFloatComparison(A=$System.Dive, B=1, Comparison Type=::NewEnumerator4)
      True Float=1.5
      False Float=0.8
     )
    )
    Module.Param0WriteEnabled=true
    Module.Param1WriteEnabled=true
    Module.Param2WriteEnabled=false
    Module.Param3WriteEnabled=true
   )
   &ParticleState()
   &ScaleColor(
    Linear Color Curve=DataInterface() {propertyValues {
      AlphaCurve {keys [<at 0.719 1>,<at 1 0>]}
      BlueCurve {keys [<at 0 0.2>,<at 0.3679 0.4>,<at 0.9715 0.4>]}
      GreenCurve {keys [<at 0 0.0352>,<at 0.3679 0.0727>,<at 0.9715 0.1646>]}
      RedCurve {keys [<at 0 0.0391>,<at 0.3679 0.079>,<at 0.9715 0.2438>]}
     }}
    Scale Mode=::NewEnumerator2
   )
  ]
  rendererClasses [&NiagaraRibbonRendererProperties]
  renderers [{rendererClass=&NiagaraRibbonRendererProperties}]
 }
 Rock_Moon_01001 {
  emitterUpdateScript [
   &EmitterState(Delay First Loop Only=true, Life Cycle Mode=::NewEnumerator1, Loop Delay=0.2, Loop Duration=1, UseLoopDelay=true)
   &SpawnRate(SpawnRate=3, Use Spawn Probability=false)
  ]
  particleSpawnScript [
   &InitializeParticle(
    Color Mode=::NewEnumerator0
    Lifetime=0.7
    Mesh Scale Mode=::NewEnumerator1
    Mesh Uniform Scale Max=0.1
    Mesh Uniform Scale Min=0.05
    Sprite Size Mode=::NewEnumerator3
    UsePositionOffset=false
    Write Material Random=true
   )
   /SetVariables(Particles.Size_Sprite:Vector2f=<20 20>) {isSetParameters=true, module=SetVariables_#7}
   &FN_SetSpawnGroup(Number of Spawn Groups=2)
   &SkeletalMeshLocation(
    Bone / Socket Index=$Particles.SpawnGroup
    Bone Sampling Mode=::NewEnumerator4
    MaskSpawnGroup=false
    Orientation Sampling=::NewEnumerator1
    Skeletal Mesh=$User.SkeletalMesh
    Write To Particle Position=false
   )
   &InitialMeshOrientation(Use Rotation Vector=false)
   /SetVariables(Particles.SpawnPosition:NiagaraPosition=&AddVectorToPosition(
     Position=$Particles.SkeletalMeshLocation.SampledPosition
     Vector=&FN_FrontEndVector3Modifier(
      Gameplay=&Lerp_Vector(
       A=&RandomRangeVector(Minimum=<-25 10 150>, Maximum=<25 25 200>)
       B=&RandomRangeVector(Minimum=<-50 100 150>, Maximum=<50 50 200>)
       Alpha=$System.Dive
      )
      Front End=&Lerp_Vector(
       A=&RandomRangeVector(Minimum=<-25 -10 150>, Maximum=<25 -25 200>)
       B=&RandomRangeVector(Minimum=<-50 -250 150>, Maximum=<50 -500 200>)
       Alpha=$User.ParaGlideLeanAlpha
      )
     )
    )) {isSetParameters=true, module=SetVariables_#8}
  ]
  particleUpdateScript [
   &SolveForcesAndVelocity(Clamp Velocity=false, Limit Acceleration=false)
   &SkeletalMeshLocation(
    Bone / Socket Index=$Particles.SpawnGroup
    Bone Sampling Mode=::NewEnumerator4
    MaskSpawnGroup=false
    Orientation Sampling=::NewEnumerator1
    Sampled Position Offset Coordinate Space:ENiagaraSkelMeshTransforms=::NewEnumerator2
    Skeletal Mesh=$User.SkeletalMesh
    Write To Particle Position=false
   ) {module=SkeletalMeshLocation001}
   /SetVariables(Particles.Position=&LerpPosition(
     A=$Particles.SpawnPosition
     B=$Particles.SkeletalMeshLocation001.SampledPosition
     Alpha=&Add_Float(
      A=$Particles.NormalizedAge
      B=&FN_FrontEndFloatModifier(
       Gameplay=&RandomRangeFloat(Minimum=-0.15, Maximum=0)
       Front End=&Lerp_Float(A=0.1, B=0, Alpha=$User.ParaGlideLeanAlpha)
      )
     )
    )) {isSetParameters=true, module=SetVariables_#9}
   &ParticleState()
   &UpdateMeshOrientation(
    Rotation Rate=&Multiply_Float(
     A=&Divide_Float(A=5, B=$Particles.Mass)
     B=&RandomRangeFloat(Minimum=0.3, Maximum=0.5)
    )
    Rotation Vector=&RandomVector(Vector Scale=1)
   )
   /SetVariables(Particles.Color_Sprite:LinearColor=&ScaleLinearColorByCurve(
     InitialValue=$Particles.Initial.Color_Sprite
     ScaleCurve=DataInterface() {propertyValues {
       AlphaCurve {keys [<at 0.0069 0.5>,<at 0.2467 0.8>,<at 0.4952 0.5>,<at 0.7503 0.8>,<at 1 0.5>]}
       BlueCurve {keys [<at 0 0>,<at 0.4777 0.4427>,<at 0.9954 0.702>]}
       GreenCurve {keys [<at 0 0.1196>,<at 0.4777 0.1671>,<at 0.9954 0.0549>]}
       RedCurve {keys [<at 0 0.5677>,<at 0.4777 0.3854>,<at 0.9954 0.1412>]}
      }}
     CurveIndex=$Particles.NormalizedAge
    )) {isSetParameters=true}
   /SetVariables(Particles.Size_Sprite:Vector2f=&Multiply_Vector2D_ByFloat(
     Float=&FloatFromCurve(
      FloatCurve=DataInterface() {propertyValues {Curve {keys [
          <at 0 0 RCIM_Cubic>
          {arriveTangent=0.3333, interpMode=RCIM_Cubic, leaveTangent=0.3333, time=0.2, value=0.5}
          {arriveTangent=-0.2898, interpMode=RCIM_Cubic, leaveTangent=-0.2898, tangentMode=RCTM_User, time=0.9, value=0.3}
          <at 1 0.3>
         ]}}}
      Scale Curve=1.5
     )
     Vector2D=$Particles.Initial.Size_Sprite
    )) {isSetParameters=true}
   &ScaleMeshSize(Scale Factor=&Multiply_VectorByFloat(
     Vector=$Engine.Owner.Scale
     Float=&FloatFromCurve(
      FloatCurve=DataInterface() {propertyValues {Curve {keys [
          <at 0 0 RCIM_Cubic>
          {arriveTangent=0.3333, interpMode=RCIM_Cubic, leaveTangent=0.3333, time=0.2, value=0.5}
          {arriveTangent=-0.2898, interpMode=RCIM_Cubic, leaveTangent=-0.2898, tangentMode=RCTM_User, time=0.9, value=0.3}
          <at 1 0.3>
         ]}}}
      Scale Curve=1.5
     )
    ))
   &GenerateDeathEvent(Use Event Probability=false) {enabled=false}
   &DynamicMaterialParameters(
    Index 3 Param 1=&FloatFromCurve(FloatCurve=DataInterface() {propertyValues {Curve {keys [
         {interpMode=RCIM_Cubic, tangentMode=RCTM_User, time=0, value=0}
         {arriveTangent=4.4532, interpMode=RCIM_Cubic, leaveTangent=4.4532, tangentMode=RCTM_User, time=1, value=2}
        ]}}})
    Index 3 Param 3=&FloatFromCurve(
     FloatCurve=DataInterface() {propertyValues {Curve {keys [
         {interpMode=RCIM_Cubic, tangentMode=RCTM_User, time=0, value=1}
         {arriveTangent=-2.4162, interpMode=RCIM_Cubic, leaveTangent=-2.4162, tangentMode=RCTM_User, time=1, value=0.2}
        ]}}}
     Scale Curve=5
    )
    Module.Param0WriteEnabled3=true
    Module.Param1WriteEnabled3=false
    Module.Param2WriteEnabled3=true
    Module.Param3WriteEnabled3=true
    Write Parameter Index 0=false
    Write Parameter Index 3=true
   )
  ]
  rendererClasses [&NiagaraMeshRendererProperties, &NiagaraSpriteRendererProperties]
  renderers [
   {rendererClass=&NiagaraMeshRendererProperties, rendererIndex=0}
   {rendererClass=&NiagaraSpriteRendererProperties, rendererIndex=1}
  ]
 }
 Rock_Moons {
  emitterUpdateScript [
   &EmitterState()
   &SpawnRate(SpawnRate=6, Use Spawn Probability=false)
  ]
  particleSpawnScript [
   &InitializeParticle(
    Color=<rgba 1 1 1 3>
    Lifetime=0.7
    Mesh Scale Mode=::NewEnumerator1
    Mesh Uniform Scale Max=0.2
    Mesh Uniform Scale Min=0.1
    Sprite Size Mode=::NewEnumerator3
    UsePositionOffset=false
    Write Material Random=true
   )
   /SetVariables(Particles.Color_Sprite:LinearColor=<rgba 0.1423 0.0555 0.7 1>, Particles.Size_Sprite:Vector2f=<17 17>) {isSetParameters=true, module=SetVariables_#7}
   &FN_SetSpawnGroup(Number of Spawn Groups=2) {enabled=false}
   &PartitionParticles()
   &SkeletalMeshLocation(
    Bone / Socket Index=$Particles.PartitionParticles.Partition
    Bone Sampling Mode=::NewEnumerator4
    MaskSpawnGroup=false
    Orientation Sampling=::NewEnumerator1
    Skeletal Mesh=$User.SkeletalMesh
    Write To Particle Position=false
   )
   &InitialMeshOrientation(Use Rotation Vector=false)
   /SetVariables(Particles.SpawnPosition:NiagaraPosition=&AddVectorToPosition(
     Position=$Particles.SkeletalMeshLocation.SampledPosition
     Vector=&FN_FrontEndVector3Modifier(
      Gameplay=&Lerp_Vector(
       A=&RandomRangeVector(Minimum=<-25 10 125>, Maximum=<25 25 175>)
       B=&RandomRangeVector(Minimum=<-50 100 125>, Maximum=<50 50 175>)
       Alpha=$System.Dive
      )
      Front End=&Lerp_Vector(
       A=&RandomRangeVector(Minimum=<-25 -10 125>, Maximum=<25 -25 175>)
       B=&RandomRangeVector(Minimum=<-50 -250 125>, Maximum=<50 -500 175>)
       Alpha=$User.ParaGlideLeanAlpha
      )
     )
    )) {isSetParameters=true, module=SetVariables_#8}
  ]
  particleUpdateScript [
   &SolveForcesAndVelocity(Clamp Velocity=false, Limit Acceleration=false)
   &SkeletalMeshLocation(
    Bone / Socket Index=$Particles.PartitionParticles.Partition
    Bone Sampling Mode=::NewEnumerator4
    MaskSpawnGroup=false
    Orientation Sampling=::NewEnumerator1
    Sampled Position Offset Coordinate Space:ENiagaraSkelMeshTransforms=::NewEnumerator2
    Skeletal Mesh=$User.SkeletalMesh
    Write To Particle Position=false
   ) {module=SkeletalMeshLocation001}
   &OrientMeshToVector(
    Look At Direction:Vector3f=&NormalizeVector(
     Vector To Normalize=&TransformVector(Vector=$Engine.Owner.Velocity, Destination Space=::Local)
     Safe Normalize=true
     Safe Fallback Threshold=0
     Safe Fallback Vector=<0 0 0.001>
    )
    Override Incoming Orientation=1
    Override Quat=false
    Reference Axis Vector:Vector3f=<-1 0 0>
    UseOptionalSideAxis=false
    UseOptionalUpAxis=false
   )
   /SetVariables(Particles.Position=&LerpPosition(
     A=$Particles.SpawnPosition
     B=&AddVectorToPosition(
      Position=$Particles.SkeletalMeshLocation001.SampledPosition
      Vector=&Lerp_Vector(A=<0 10 0>, B=<-10 0 10>, Alpha=$User.ParaGlideLeanAlpha)
     )
     Alpha=&Add_Float(
      A=$Particles.NormalizedAge
      B=&FN_FrontEndFloatModifier(
       Gameplay=&RandomRangeFloat(Minimum=-0.15, Maximum=0)
       Front End=&Lerp_Float(A=0.1, B=0, Alpha=$User.ParaGlideLeanAlpha)
      )
     )
    )) {isSetParameters=true, module=SetVariables_#9}
   &ParticleState()
   &UpdateMeshOrientation(
    Rotation Rate=&Multiply_Float(
     A=&Divide_Float(A=5, B=$Particles.Mass)
     B=&RandomRangeFloat(Minimum=0.3, Maximum=0.5)
    )
    Rotation Vector=&RandomVector(Vector Scale=1)
   )
   /SetVariables(Particles.Color_Sprite:LinearColor=&ScaleLinearColorByCurve(
     InitialValue=$Particles.Initial.Color_Sprite
     ScaleCurve=DataInterface() {propertyValues {
       AlphaCurve {keys [<at 0.2958 0>,<at 0.5459 1>,<at 1 0>]}
       BlueCurve {keys [<at 0 1>,<at 0.2649 1>,<at 1 1>]}
       GreenCurve {keys [<at 0 0.54>,<at 0.2649 1>,<at 1 0>]}
       RedCurve {keys [<at 0 0>,<at 0.2649 1>,<at 1 0.509>]}
      }}
     CurveIndex=$Particles.NormalizedAge
    )) {isSetParameters=true}
   &ScaleColor(
    Linear Color Curve=DataInterface() {propertyValues {
      AlphaCurve {keys [<at 0 0>,<at 0.4954 1>,<at 0.9739 0>]}
      BlueCurve {keys [<at 0 1>,<at 0.5074 1>,<at 0.998 1>]}
      GreenCurve {keys [<at 0 0.4323>,<at 0.5074 0.5366>,<at 0.998 0.4323>]}
      RedCurve {keys [<at 0 0.6783>,<at 0.5074 0.4385>,<at 0.998 0.6783>]}
     }}
    Scale Mode=::NewEnumerator2
   ) {module=ScaleColor001}
   &ScaleSpriteSize(
    Initial Sprite Size:Vector2f=$Particles.Initial.Size_Sprite
    Uniform Curve Scale=1.5
    Uniform Curve Sprite Scale=DataInterface() {propertyValues {Curve {keys [
        {arriveTangent=-0.0548, interpMode=RCIM_Cubic, leaveTangent=-0.0548, tangentMode=RCTM_User, time=0, value=0}
        {interpMode=RCIM_Cubic, tangentMode=RCTM_User, time=0.5, value=1}
        {arriveTangent=-3.2659, interpMode=RCIM_Cubic, leaveTangent=-3.2659, tangentMode=RCTM_User, time=1, value=0.25}
       ]}}}
   )
   &ScaleMeshSize(Scale Factor=&Multiply_VectorByFloat(
     Vector=$Engine.Owner.Scale
     Float=&FloatFromCurve(
      FloatCurve=DataInterface() {propertyValues {Curve {keys [
          <at 0 0 RCIM_Cubic>
          {arriveTangent=0.3333, interpMode=RCIM_Cubic, leaveTangent=0.3333, time=0.2, value=0.5}
          {arriveTangent=-0.2898, interpMode=RCIM_Cubic, leaveTangent=-0.2898, tangentMode=RCTM_User, time=0.9, value=0.3}
          <at 1 0.3>
         ]}}}
      Scale Curve=1.5
     )
    ))
   &DynamicMaterialParameters(
    Index 3 Param 1=&FloatFromCurve(FloatCurve=DataInterface() {propertyValues {Curve {keys [
         {interpMode=RCIM_Cubic, tangentMode=RCTM_User, time=0, value=0}
         {arriveTangent=4.4532, interpMode=RCIM_Cubic, leaveTangent=4.4532, tangentMode=RCTM_User, time=1, value=2}
        ]}}})
    Index 3 Param 3=&FloatFromCurve(
     FloatCurve=DataInterface() {propertyValues {Curve {keys [
         {interpMode=RCIM_Cubic, tangentMode=RCTM_User, time=0, value=1}
         {arriveTangent=-2.4162, interpMode=RCIM_Cubic, leaveTangent=-2.4162, tangentMode=RCTM_User, time=1, value=0.2}
        ]}}}
     Scale Curve=5
    )
    Module.Param0WriteEnabled3=true
    Module.Param1WriteEnabled3=false
    Module.Param2WriteEnabled3=true
    Module.Param3WriteEnabled3=true
    Write Parameter Index 0=false
    Write Parameter Index 3=true
   )
  ]
  rendererClasses [&NiagaraMeshRendererProperties, &NiagaraSpriteRendererProperties]
  renderers [
   {rendererClass=&NiagaraMeshRendererProperties, rendererIndex=0}
   {rendererClass=&NiagaraSpriteRendererProperties, rendererIndex=1}
  ]
 }
 Rocks {
  emitterUpdateScript [
   &EmitterState(
    Life Cycle Mode=::NewEnumerator1
    Loop Duration=&RandomRangeFloat(Minimum=0.2, Maximum=0.8)
    UseLoopDelay=false
   )
   &SpawnRate(
    SpawnGroup:NiagaraInt32=&RandomRangeInt(Minimum=0, Maximum=1, Evaluation Type=::NewEnumerator1)
    SpawnRate=10
    Use Spawn Probability=false
   )
   &SpawnBurst_Instantaneous(Spawn Count=3, Spawn Group=1, Use Spawn Probability=false, UseLoopCountLimit=false) {module=SpawnBurst_Instantaneous001}
   &SpawnBurst_Instantaneous(Spawn Count=3, Spawn Time=0.2, Use Spawn Probability=false, UseLoopCountLimit=false)
  ]
  particleSpawnScript [
   &InitializeParticle(
    Color Maximum=<rgba 0.0802 0.0823 0.0802 1>
    Color Minimum=<rgba 0.0497 0.0482 0.0497 1>
    Color Mode=::NewEnumerator2
    Lifetime Max=0.5
    Lifetime Min=0.4
    Lifetime Mode=::NewEnumerator1
    Mesh Scale Max=<0.05 0.05 0.05>
    Mesh Scale Min=<0.025 0.025 0.025>
    Mesh Scale Mode=::NewEnumerator2
    UsePositionOffset=false
    Write Material Random=true
   )
   &InitialMeshOrientation(
    Rotation=&RandomRangeVector(Minimum=<0 0 0>, Maximum=<1 1 1>)
    Use Orientation Vector=true
    Use Rotation Vector=true
   )
   &PartitionParticles() {enabled=false}
   /SetVariables(Particles.SpawnGroup=$Emitter.SpawnGroup) {isSetParameters=true}
   &SkeletalMeshLocation(
    Bone / Socket Index=$Particles.SpawnGroup
    Bone Sampling Mode=::NewEnumerator4
    Interpolate Mesh Sampling=false
    MaskSpawnGroup=false
    Orientation Sampling=::NewEnumerator1
    Sampled Position Offset=&MergeHelper_VectorSwitch(0=<5 -10 0>, 1=<-5 10 0>, 2=?, 3=?, Index=$Particles.SpawnGroup, Num=::NewEnumerator0)
    Skeletal Mesh=$User.SkeletalMesh
   )
   &SphereLocation(
    MaskSpawnGroup=false
    Override Seed=false
    Sphere Coordinate Space=::Simulation
    Sphere Origin:NiagaraPosition=$StackContext.Position
    Sphere Radius=3
    Surface Only=false
    UseNonUniformScale=false
    UseOffset=false
    UseSphereOrientation=false
   )
   &AddVelocity(Velocity=&Multiply_Vector(
     A=$Output.SphereLocation.SphereNormal
     B=&VectorFromFloat(Value=&RandomRangeFloat(Minimum=150, Maximum=100))
    ))
  ]
  particleUpdateScript [
   &MeshRotationForce(Rotation=&RandomRangeVector(Minimum=<-10 -10 -10>, Maximum=<10 10 10>))
   &SolveForcesAndVelocity(Clamp Velocity=false, Limit Acceleration=false)
   &ParticleState()
   &ScaleMeshSize(Scale Factor=&VectorFromFloat(Value=&FloatFromCurve(
      FloatCurve=DataInterface() {propertyValues {Curve {keys [<at 0 1>,<at 0.5 1>,<at 1 0>]}}}
      Scale Curve=&Lerp_Float(A=0.25, B=0.25, Alpha=$System.Dive)
     )))
   &InterpolateOverTime(
    Initial Vector Value:NiagaraPosition=$Particles.Previous.Position
    Target Vector Value:Vector3f=$Particles.SkeletalMeshLocation.SampledPosition
    Variable Type=::NewEnumerator2
   )
   /SetVariables(Particles.Position=/Niagara/DynamicInputs/Lerp/SmoothLerpOverTime_Position(
     Convergence Rate:NiagaraFloat=6
     Smooth Value:NiagaraPosition=$Particles.Position
     Target Value:NiagaraPosition=$Particles.InterpolateOverTime.InterpolateOverTime_V3_.Moving Average
    )) {isSetParameters=true}
   &DynamicMaterialParameters(
    Index 3 Param 3=2.5
    Module.Param0WriteEnabled3=false
    Module.Param1WriteEnabled3=false
    Module.Param2WriteEnabled3=true
    Module.Param3WriteEnabled3=false
    Write Parameter Index 0=false
    Write Parameter Index 3=true
   )
  ]
  rendererClasses [&NiagaraMeshRendererProperties]
  renderers [{rendererClass=&NiagaraMeshRendererProperties, rendererIndex=0}]
 }
 Sparkles001 {
  emitterUpdateScript [
   &EmitterState()
   &SpawnRate(SpawnRate=&RandomRangeFloat(Minimum=30, Maximum=60), Use Spawn Probability=false)
  ]
  particleSpawnScript [
   &InitializeParticle(
    Color Maximum=<rgba 0.0936 0.0089 0.8021 1>
    Color Minimum=<rgba 0.4062 0.1355 0.651 1>
    Color Mode=::NewEnumerator2
    Lifetime=&FloatFromCurve(
     FloatCurve=DataInterface() {propertyValues {Curve {keys [
         {arriveTangent=4.1785, interpMode=RCIM_Cubic, leaveTangent=4.1785, tangentMode=RCTM_User, time=0, value=0}
         {arriveTangent=1.1179, interpMode=RCIM_Cubic, leaveTangent=1.1179, tangentMode=RCTM_User, time=0.2042, value=0.6228}
         {arriveTangent=0.0224, interpMode=RCIM_Cubic, leaveTangent=0.0224, tangentMode=RCTM_User, time=1, value=1}
        ]}}}
     CurveIndex=$System.NormalizedLoopAge
    )
    Mesh Scale Mode=::NewEnumerator4
    Position Mode=::NewEnumerator1
    Sprite Rotation Mode=::NewEnumerator2
    Sprite Size Mode=::NewEnumerator1
    Uniform Sprite Size Max=5
    Uniform Sprite Size Min=2.5
    UsePositionOffset=false
    Write Material Random=true
   )
   &FN_SetSpawnGroup(Number of Spawn Groups=2)
   &SkeletalMeshLocation(
    Bone / Socket Index=$Particles.SpawnGroup
    Bone Sampling Mode=::NewEnumerator4
    MaskSpawnGroup=false
    Orientation Sampling=::NewEnumerator1
    Skeletal Mesh=$User.SkeletalMesh
    Write To Particle Position=false
   ) {module=SkeletalMeshLocation002}
   /SetVariables(Particles.SpawnPosition:NiagaraPosition=&AddVectorToPosition(
     Position=$Particles.SkeletalMeshLocation002.SampledPosition
     Vector=&FN_FrontEndVector3Modifier(
      Gameplay=&Lerp_Vector(
       A=&RandomRangeVector(Minimum=<-25 10 50>, Maximum=<25 25 125>)
       B=&RandomRangeVector(Minimum=<-50 100 50>, Maximum=<50 50 125>)
       Alpha=$System.Dive
      )
      Front End=&Lerp_Vector(
       A=&RandomRangeVector(Minimum=<-25 -10 50>, Maximum=<25 -25 125>)
       B=&RandomRangeVector(Minimum=<-50 -250 50>, Maximum=<50 -500 125>)
       Alpha=$User.ParaGlideLeanAlpha
      )
     )
    )) {isSetParameters=true}
  ]
  particleUpdateScript [
   &SkeletalMeshLocation(
    Bone / Socket Index=$Particles.SpawnGroup
    Bone Sampling Mode=::NewEnumerator4
    MaskSpawnGroup=false
    Orientation Sampling=::NewEnumerator1
    Skeletal Mesh=$User.SkeletalMesh
    Write To Particle Position=false
   ) {module=SkeletalMeshLocation001}
   /SetVariables(Particles.Position=&LerpPosition(
     A=$Particles.SpawnPosition
     B=$Particles.SkeletalMeshLocation001.SampledPosition
     Alpha=&Add_Float(
      A=$Particles.NormalizedAge
      B=&FN_FrontEndFloatModifier(
       Gameplay=&RandomRangeFloat(Minimum=-0.15, Maximum=0)
       Front End=&Lerp_Float(A=0.1, B=0, Alpha=$User.ParaGlideLeanAlpha)
      )
     )
    )) {isSetParameters=true}
   &ParticleState()
   &DynamicMaterialParameters(
    Index 0 Param 1=1
    Index 0 Param 2=1
    Index 0 Param 3=0.5
    Module.Param0WriteEnabled=true
    Module.Param1WriteEnabled=true
    Module.Param2WriteEnabled=true
    Module.Param3WriteEnabled=true
   )
   &ScaleColor(
    Linear Color Curve=DataInterface() {propertyValues {
      AlphaCurve {keys [<at 0 1>,<at 1 0>]}
      BlueCurve {keys [<at 0 5>,<at 0.1051 1>,<at 0.2645 0.9531>,<at 0.517 0.7>]}
      GreenCurve {keys [<at 0 5>,<at 0.1051 1>,<at 0.2645 0.9531>,<at 0.517 0.7>]}
      RedCurve {keys [<at 0 5>,<at 0.1051 1>,<at 0.2645 0.9531>,<at 0.517 0.7>]}
     }}
    Scale Mode=::NewEnumerator2
   )
   &ScaleSpriteSize(Scale Factor=&Vector2DFromFloat(Value=&FloatFromCurve(FloatCurve=DataInterface() {propertyValues {Curve {keys [
          {arriveTangent=14.26, interpMode=RCIM_Cubic, leaveTangent=14.26, tangentMode=RCTM_User, time=0, value=0}
          {arriveTangentWeight=0.0869, interpMode=RCIM_Cubic, leaveTangentWeight=0.1958, tangentMode=RCTM_User, time=0.2, value=1}
          <at 1 0>
         ]}}})))
  ]
  rendererClasses [&NiagaraSpriteRendererProperties]
  renderers [{rendererClass=&NiagaraSpriteRendererProperties}]
 }
}
summary {
 systemName=NS_Contrail_Nebula
 emitters [
  {
   emitterName=Hand_Rocks
   rendererClasses [
    &NiagaraMeshRendererProperties
    &NiagaraSpriteRendererProperties
    &NiagaraLightRendererProperties
   ]
  }
  {emitterName=Sparkles001, rendererClasses [&NiagaraSpriteRendererProperties]}
  {emitterName=Flames_Attached_to_Rocks002, rendererClasses [&NiagaraRibbonRendererProperties]}
  {
   emitterName=Rock_Moon_01001
   rendererClasses [&NiagaraMeshRendererProperties, &NiagaraSpriteRendererProperties]
  }
  {emitterName=Rock_Moons, rendererClasses [&NiagaraMeshRendererProperties, &NiagaraSpriteRendererProperties]}
  {emitterName=Flames_Attached_to_Rocks, rendererClasses [&NiagaraRibbonRendererProperties]}
  {emitterName=Rocks, rendererClasses [&NiagaraMeshRendererProperties]}
  {emitterName=MergedRibbons, rendererClasses [&NiagaraRibbonRendererProperties]}
 ]
 userVariables [
  User.bIsFrontEnd {type=NiagaraBool}
  User.bIsFrontEndPreview {type=NiagaraBool}
  User.FeetRestingPositions {
   type=NiagaraDataInterfaceArrayFloat3
   description=""
   defaultValue {
    struct=NiagaraExt_VariableValue_DataInterface
    value {
     dataInterface=&self:NiagaraDataInterfaceArrayFloat3_0
     dataInterfaceClass=&NiagaraDataInterfaceArrayFloat3
    }
   }
  }
  User.ForwardDot {type=NiagaraFloat, defaultValue=0, description=""}
  User.Landing {type=NiagaraBool}
  User.ParaGlideLeanAlpha {type=NiagaraFloat, defaultValue=0, description=""}
  User.PlayerSpeed {type=NiagaraFloat, defaultValue=0, description=""}
  User.RibbonIDOffset {type=NiagaraInt32}
  User.SkeletalMesh {
   type=NiagaraDataInterfaceSkeletalMesh
   defaultValue {
    struct=NiagaraExt_VariableValue_DataInterface
    value {dataInterface=&self:NiagaraDataInterfaceSkeletalMesh_0}
   }
  }
  User.SpawnCount {type=NiagaraInt32}
 ]
}
